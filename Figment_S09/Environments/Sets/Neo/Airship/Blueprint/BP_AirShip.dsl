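; Airship set-piece actor that circles a pivot offset to its side and plays ambient audio.
; The construction script places the RotatingMovement pivot at RotationRadius on local Y and derives a negative yaw rate from Speed divided by the radius, so larger orbits turn slower.
; It also draws a temporary red debug circle previewing the orbit path.
; BeginPlay starts the ambient audio unless the IsInFrontend flag is set.
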(event EventBeginPlay
  (if (not (Variables|Default|GetIsInFrontend))
    (Audio|Components|Audio|Play (Variables|Default|GetAmbientAudio) 0.0)))

(fn ConstructionScript
  (bind _rotationradius (Variables|Default|GetRotationRadius))
  (Variables|/Script/Engine.RotatingMovementComponent|SetPivotTranslation (Math|Vector|MakeVector 0.0 _rotationradius) (Variables|Default|GetRotatingMovement))
  (Variables|/Script/Engine.RotatingMovementComponent|SetRotationRate (Math|Rotator|MakeRotator 0.0 0.0 (/ (* (Variables|Default|GetSpeed) -1e+04) (* _rotationradius 2.0))) (Variables|Default|GetRotatingMovement))
  (Rendering|Debug|DrawDebugCircle (+ (Transformation|GetActorLocation) (* (Math|Vector|GetRightVector (Transformation|GetActorRotation)) (Variables|Default|GetRotationRadius))) (Variables|Default|GetRotationRadius) 64 "(R=1.000000,G=0.000000,B=0.000000,A=1.000000)" 1e+01 5e+01 "0.000000,1.000000,0.000000" "1.000000,0.000000,0.000000"))
