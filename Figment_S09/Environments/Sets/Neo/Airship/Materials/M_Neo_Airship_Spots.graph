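# Additive unlit material for the Neo airship's spots, drawing soft, tinted light shafts that sway over time.
# Two sines at separate speeds, randomized per element via Pivot Painter, drive object-scaled Y and Z offsets weighted by UV V, added to a reused PartyBus animation WPO function.
# Opacity combines inverted Fresnel and a UV falloff with depth, near-camera and cull-distance fades.
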Material /Figment_S09/Environments/Sets/Neo/Airship/Materials/M_Neo_Airship_Spots
graph {
 properties {blendMode=BLEND_Additive, pixelDepthOffsetMode=PDOM_Legacy, shadingModel=MSM_Unlit, usedShadingModels=Unlit}
 nodes [
  n1 = TextureCoordinate()
  n2 = Multiply(
   A=MaterialFunctionCall() {materialFunction=/Engine/Functions/Engine_MaterialFunctions02/WorldPositionOffset/ObjectScale}.Scale X
   B=ScalarParameter() {defaultValue=500, parameterName=Movement}
  )
  n3 = Multiply(A=@n2) {constB=-1}
  n4 = Add(
   A=Time()
   B=Multiply(
    A=MaterialFunctionCall() {materialFunction=/Engine/Functions/Engine_MaterialFunctions02/PivotPainter/PivotPainter_PerObjectData}.Random Value Per Element
    B=ScalarParameter() {defaultValue=32.369, parameterName=Random}
   )
  )
  n5 = Multiply(A=@n4, B=ScalarParameter() {defaultValue=0.232, parameterName=Speed})
  n6 = AppendVector(A=Constant(), B=LinearInterpolate(A=@n2, B=@n3, Alpha=Sine(None=@n5)))
  n7 = ScalarParameter() {defaultValue=0.3, parameterName=Speed2}
  n8 = LinearInterpolate(A=@n2, B=@n3, Alpha=Sine(None=Multiply(A=@n4, B=@n7)))
  Add(
   A=MaterialFunctionCall() {materialFunction=/Game/Athena/Aircraft/PartyBus_RT/Materials/MF_PartyBusAnimationWPO}.Result
   B=Multiply(A=ComponentMask(None=@n1) {g=true}, B=AppendVector(A=@n6, B=@n8))
  )
  n9 = Fresnel(ExponentIn=ScalarParameter() {defaultValue=0.2362, parameterName=Exponent}) {baseReflectFraction=0}
  n10 = OneMinus(None=ComponentMask(None=TextureCoordinate()) {g=true})
  n11 = ScalarParameter() {defaultValue=3, parameterName=Falloff}
  n12 = DepthFade(
   Opacity=Multiply(A=OneMinus(None=@n9), B=Power(Base=@n10, Exp=@n11))
   FadeDistance=ScalarParameter() {defaultValue=500, parameterName=DepthFade}
  )
  n13 = MaterialFunctionCall(Distance=ScalarParameter() {defaultValue=500, parameterName=NearFade}) {materialFunction=&Near_Camera_Fade}
  Multiply(A=Multiply(A=Multiply(A=@n12, B=@n13.Result), B=ScalarParameter() {defaultValue=0.2143, parameterName=Opacity}), B=DistanceCullFade())
  VectorParameter() {defaultValue=<rgba 0.5249 0.7001 1 1>}
  VertexNormalWS()
 ]
}
